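# PistolCanary comic hero intro widget that animates character art across three comic panels with an intro sound.
# When the comic animation finishes, a full wipe plays and the widget removes itself after a delay; the wipe can also be triggered by event.
# A load event sets panning materials and soft textures in delayed per-panel batches, choosing low-res sets on mobile.
# Per-panel unload events clear image brushes, then force garbage collection.
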WidgetBlueprint /PistolCanary/Layout/BP/ComicPanels/HeroIntro/PistolCanary_ComicHeroIntro
animations [
 ComicHeroIntro {
  binding_count=33
  length_seconds=13.9274
  object=&self:ComicHeroIntro
  sequence {
   bindings [
    06_BG {
     id=#1
     possessed_class=Image
     tracks [{
       type=MovieScene2DTransformTrack
       sections [{
         type=MovieScene2DTransformSection
         end_frame=450
         start_frame=-3
         channels [
          Translation.X {type=MovieSceneScriptingFloatChannel, keys [[0,3432.1208],[214.9,3432.1208],[223.644,0]]}
          Translation.Y {type=MovieSceneScriptingFloatChannel, keys [[0,2301.1042],[214.9,2301.1042],[223.644,0]]}
         ]
        }]
      }]
    }
    Panel2Chr3Image {
     id=#2
     possessed_class=Image
     tracks [
      {
       type=MovieSceneFloatTrack
       display_name=Render Opacity
       sections [{
         type=MovieSceneFloatSection
         end_frame=658
         start_frame=-3
         channels [None {type=MovieSceneScriptingFloatChannel, keys [[-2.421,0],[390.554,1],[658,0]]}]
        }]
      }
      {
       type=MovieScene2DTransformTrack
       sections [{
         type=MovieScene2DTransformSection
         end_frame=767
         start_frame=0
         channels [
          Translation.X {type=MovieSceneScriptingFloatChannel, default=0}
          Translation.Y {type=MovieSceneScriptingFloatChannel, keys [[0,-2009.9679],[151.347,-2011.001],[545.255,-1839.474],[558.709,-7.0294]]}
          Angle {type=MovieSceneScriptingFloatChannel, default=0}
          Scale.X {type=MovieSceneScriptingFloatChannel, default=0.75, const [1,1]}
          Scale.Y {type=MovieSceneScriptingFloatChannel, default=0.75, const [1,1]}
          Shear.X {type=MovieSceneScriptingFloatChannel, default=0}
          Shear.Y {type=MovieSceneScriptingFloatChannel, default=0}
         ]
        }]
      }
     ]
    }
    P3Chr03 {
     id=#3
     possessed_class=Image
     tracks [{
       type=MovieScene2DTransformTrack
       sections [{
         type=MovieScene2DTransformSection
         end_frame=832
         start_frame=-3
         channels [
          Translation.X {type=MovieSceneScriptingFloatChannel, default=0, const [663.9196,1]}
          Translation.Y {type=MovieSceneScriptingFloatChannel, default=0, const [-450.2097,1]}
          Angle {type=MovieSceneScriptingFloatChannel, default=0}
          Scale.X {type=MovieSceneScriptingFloatChannel, keys [[637.498,1.25],[832.93,1.22]]}
          Scale.Y {type=MovieSceneScriptingFloatChannel, keys [[637.498,1.25],[832.93,1.22]]}
          Shear.X {type=MovieSceneScriptingFloatChannel, default=0}
          Shear.Y {type=MovieSceneScriptingFloatChannel, default=0}
         ]
        }]
      }]
    }
    Panel1 {
     id=#4
     possessed_class=CanvasPanel
     tracks [
      {
       type=MovieScene2DTransformTrack
       sections [{
         type=MovieScene2DTransformSection
         end_frame=780
         start_frame=0
         channels [
          Translation.X {type=MovieSceneScriptingFloatChannel, const [0,1]}
          Translation.Y {type=MovieSceneScriptingFloatChannel, const [0,1]}
          Angle {type=MovieSceneScriptingFloatChannel, const [0,1]}
          Scale.X {type=MovieSceneScriptingFloatChannel, keys [[0,0.5],[727.561,0.5],[767.434,5]]}
          Scale.Y {type=MovieSceneScriptingFloatChannel, keys [[0,0.5],[727.561,0.5],[767.434,5]]}
          Shear.X {type=MovieSceneScriptingFloatChannel, const [0,1]}
          Shear.Y {type=MovieSceneScriptingFloatChannel, const [0,1]}
         ]
        }]
      }
      {
       type=MovieSceneFloatTrack
       display_name=Render Opacity
       sections [{
         type=MovieSceneFloatSection
         end_frame=780
         start_frame=0
         channels [None {type=MovieSceneScriptingFloatChannel, default=1, keys [[0,1],[411.23,1],[419.664,0]]}]
        }]
      }
     ]
    }
    Panel2Chr4Image {
     id=#5
     possessed_class=Image
     tracks [
      {
       type=MovieSceneFloatTrack
       display_name=Render Opacity
       sections [{
         type=MovieSceneFloatSection
         end_frame=658
         start_frame=-3
         channels [None {type=MovieSceneScriptingFloatChannel, keys [[-2.421,0],[390.554,1],[658,0]]}]
        }]
      }
      {
       type=MovieScene2DTransformTrack
       sections [{
         type=MovieScene2DTransformSection
         end_frame=735
         start_frame=0
         channels [
          Translation.X {type=MovieSceneScriptingFloatChannel, default=0}
          Translation.Y {type=MovieSceneScriptingFloatChannel, keys [[0,-1951.1543],[151.347,-1952.6387],[457.682,-1839.474],[471.85,-7.0294]]}
          Angle {type=MovieSceneScriptingFloatChannel, default=0}
          Scale.X {type=MovieSceneScriptingFloatChannel, default=0.75, const [1,1]}
          Scale.Y {type=MovieSceneScriptingFloatChannel, default=0.75, const [1,1]}
          Shear.X {type=MovieSceneScriptingFloatChannel, default=0}
          Shear.Y {type=MovieSceneScriptingFloatChannel, default=0}
         ]
        }]
      }
     ]
    }
    04_BG {
     id=#6
     possessed_class=Image
     tracks [
      {
       type=MovieScene2DTransformTrack
       sections [{
         type=MovieScene2DTransformSection
         end_frame=734
         start_frame=-3
         channels [
          Translation.X {type=MovieSceneScriptingFloatChannel, keys [[0,-3643.2712],[205.86,-3643.2712],[206.008,3432.1208],[214.752,0],[734.782,0]]}
          Translation.Y {type=MovieSceneScriptingFloatChannel, keys [[0,2340.7009],[205.86,2340.7009],[214.752,0],[734.782,0]]}
         ]
        }]
      }
      {
       type=MovieSceneFloatTrack
       display_name=Render Opacity
       sections [{type=MovieSceneFloatSection, end_frame=419, start_frame=419, channels [None {type=MovieSceneScriptingFloatChannel, default=0, const [1,1]}]}]
      }
     ]
    }
    02_BG {
     id=#7
     possessed_class=Image
     tracks [
      {
       type=MovieScene2DTransformTrack
       sections [{
         type=MovieScene2DTransformSection
         end_frame=60
         start_frame=-3
         channels [
          Translation.X {type=MovieSceneScriptingFloatChannel, keys [[0,4252.2124],[28.586,4252.2124],[42.879,0]]}
          Translation.Y {type=MovieSceneScriptingFloatChannel, keys [[0,-2445.3645],[28.586,-2445.3645],[42.879,0]]}
         ]
        }]
      }
      {
       type=MovieSceneFloatTrack
       display_name=Render Opacity
       sections [{
         type=MovieSceneFloatSection
         end_frame=693
         start_frame=0
         channels [None {type=MovieSceneScriptingFloatChannel, default=1, keys [[0,1],[478,0]]}]
        }]
      }
     ]
    }
    01_BG {
     id=#8
     possessed_class=Image
     tracks [
      {
       type=MovieScene2DTransformTrack
       sections [{
         type=MovieScene2DTransformSection
         end_frame=90
         start_frame=-3
         channels [
          Translation.X {type=MovieSceneScriptingFloatChannel, keys [[0,-3643.2712],[14.293,-3643.2712],[28.586,0]]}
          Translation.Y {type=MovieSceneScriptingFloatChannel, keys [[0,2340.7009],[14.293,2340.7009],[28.586,0]]}
         ]
        }]
      }
      {
       type=MovieSceneFloatTrack
       display_name=Render Opacity
       sections [{
         type=MovieSceneFloatSection
         end_frame=693
         start_frame=0
         channels [None {type=MovieSceneScriptingFloatChannel, default=1, keys [[0,1],[478,0]]}]
        }]
      }
     ]
    }
    BackgroundGroup1 {
     id=#9
     possessed_class=CanvasPanel
     tracks [{
       type=MovieScene2DTransformTrack
       sections [{
         type=MovieScene2DTransformSection
         end_frame=288
         start_frame=-3
         channels [
          Translation.X {type=MovieSceneScriptingFloatChannel, const [0,1]}
          Translation.Y {type=MovieSceneScriptingFloatChannel, const [0,1]}
          Angle {type=MovieSceneScriptingFloatChannel, const [0,1]}
          Scale.X {type=MovieSceneScriptingFloatChannel, const [1,1]}
          Scale.Y {type=MovieSceneScriptingFloatChannel, const [1,1]}
          Shear.X {type=MovieSceneScriptingFloatChannel, const [0,1]}
          Shear.Y {type=MovieSceneScriptingFloatChannel, const [0,1]}
         ]
        }]
      }]
    }
    Panel3 {
     id=#10
     possessed_class=CanvasPanel
     tracks [
      {
       type=MovieScene2DTransformTrack
       sections [{
         type=MovieScene2DTransformSection
         end_frame=657
         start_frame=-3
         channels [
          Translation.X {type=MovieSceneScriptingFloatChannel, keys [[0,-1920],[637.391,-1920],[657.735,0]]}
          Translation.Y {type=MovieSceneScriptingFloatChannel, const [0,2]}
          Angle {type=MovieSceneScriptingFloatChannel, default=0}
          Scale.X {type=MovieSceneScriptingFloatChannel, default=1, const [1,1]}
          Scale.Y {type=MovieSceneScriptingFloatChannel, default=1, const [1,1]}
          Shear.X {type=MovieSceneScriptingFloatChannel, default=0}
          Shear.Y {type=MovieSceneScriptingFloatChannel, default=0}
         ]
        }]
      }
      {
       type=MovieSceneFloatTrack
       display_name=Render Opacity
       sections [{
         type=MovieSceneFloatSection
         end_frame=833
         start_frame=-3
         channels [None {type=MovieSceneScriptingFloatChannel, keys [[0,0],[638,1],[822.502,1],[833.02,0]]}]
        }]
      }
     ]
    }
    P3Chr02 {
     id=#11
     possessed_class=Image
     tracks [{
       type=MovieScene2DTransformTrack
       sections [{
         type=MovieScene2DTransformSection
         end_frame=832
         start_frame=-3
         channels [
          Translation.X {type=MovieSceneScriptingFloatChannel, default=0, const [-244.4598,1]}
          Translation.Y {type=MovieSceneScriptingFloatChannel, default=0, const [-426.539,1]}
          Angle {type=MovieSceneScriptingFloatChannel, default=0}
          Scale.X {type=MovieSceneScriptingFloatChannel, keys [[637.498,1.1],[832.93,1.07]]}
          Scale.Y {type=MovieSceneScriptingFloatChannel, keys [[637.498,1.1],[832.93,1.07]]}
          Shear.X {type=MovieSceneScriptingFloatChannel, default=0}
          Shear.Y {type=MovieSceneScriptingFloatChannel, default=0}
         ]
        }]
      }]
    }
    Panel2Chr1Image {
     id=#12
     possessed_class=Image
     tracks [
      {
       type=MovieSceneFloatTrack
       display_name=Render Opacity
       sections [{
         type=MovieSceneFloatSection
         end_frame=658
         start_frame=-3
         channels [None {type=MovieSceneScriptingFloatChannel, keys [[-2.421,0],[390.554,1],[658,0]]}]
        }]
      }
      {
       type=MovieScene2DTransformTrack
       sections [{
         type=MovieScene2DTransformSection
         end_frame=727
         start_frame=0
         channels [
          Translation.X {type=MovieSceneScriptingFloatChannel, default=0}
          Translation.Y {type=MovieSceneScriptingFloatChannel, keys [[0,-1839.474],[415.615,-1839.474],[431.93,-7.0294]]}
          Angle {type=MovieSceneScriptingFloatChannel, default=0}
          Scale.X {type=MovieSceneScriptingFloatChannel, default=0.75, const [1,1]}
          Scale.Y {type=MovieSceneScriptingFloatChannel, default=0.75, const [1,1]}
          Shear.X {type=MovieSceneScriptingFloatChannel, default=0}
          Shear.Y {type=MovieSceneScriptingFloatChannel, default=0}
         ]
        }]
      }
     ]
    }
    02_Chr {
     id=#13
     possessed_class=Image
     tracks [
      {
       type=MovieSceneFloatTrack
       display_name=Render Opacity
       sections [{
         type=MovieSceneFloatSection
         end_frame=693
         start_frame=0
         channels [None {type=MovieSceneScriptingFloatChannel, default=1, keys [[0,1],[478,0]]}]
        }]
      }
      {
       type=MovieScene2DTransformTrack
       sections [{
         type=MovieScene2DTransformSection
         end_frame=288
         start_frame=-3
         channels [
          Translation.X {type=MovieSceneScriptingFloatChannel, keys [[116.278,4224.0908],[131.595,880.1874]]}
          Translation.Y {type=MovieSceneScriptingFloatChannel, keys [[116.278,-1803.0375],[131.595,239.3695]]}
          Angle {type=MovieSceneScriptingFloatChannel, default=0}
          Scale.X {type=MovieSceneScriptingFloatChannel, default=1, keys [[131.595,1.35],[288.693,1.25]]}
          Scale.Y {type=MovieSceneScriptingFloatChannel, default=1, keys [[131.595,1.35],[288.693,1.25]]}
          Shear.X {type=MovieSceneScriptingFloatChannel, default=0}
          Shear.Y {type=MovieSceneScriptingFloatChannel, default=0}
         ]
        }]
      }
     ]
    }
    01_ChrA {
     id=#14
     possessed_class=Image
     tracks [
      {
       type=MovieSceneFloatTrack
       display_name=Render Opacity
       sections [{
         type=MovieSceneFloatSection
         end_frame=693
         start_frame=0
         channels [None {type=MovieSceneScriptingFloatChannel, default=1, keys [[0,1],[161.777,1],[478,0]]}]
        }]
      }
      {
       type=MovieScene2DTransformTrack
       sections [{
         type=MovieScene2DTransformSection
         end_frame=66
         start_frame=-3
         channels [
          Translation.X {type=MovieSceneScriptingFloatChannel, keys [[46.475,-2296.4412],[66.104,102.1673]]}
          Translation.Y {type=MovieSceneScriptingFloatChannel, keys [[46.475,1560.0889],[66.104,92.0813]]}
          Angle {type=MovieSceneScriptingFloatChannel, const [0,1]}
          Scale.X {type=MovieSceneScriptingFloatChannel, const [1.1,1]}
          Scale.Y {type=MovieSceneScriptingFloatChannel, const [1.1,1]}
          Shear.X {type=MovieSceneScriptingFloatChannel, default=0}
          Shear.Y {type=MovieSceneScriptingFloatChannel, default=0}
         ]
        }]
      }
     ]
    }
    Panel2Chr5Image {
     id=#15
     possessed_class=Image
     tracks [
      {
       type=MovieSceneFloatTrack
       display_name=Render Opacity
       sections [{
         type=MovieSceneFloatSection
         end_frame=658
         start_frame=-3
         channels [None {type=MovieSceneScriptingFloatChannel, keys [[-2.421,0],[390.554,1],[658,0]]}]
        }]
      }
      {
       type=MovieScene2DTransformTrack
       sections [{
         type=MovieScene2DTransformSection
         end_frame=739
         start_frame=0
         channels [
          Translation.X {type=MovieSceneScriptingFloatChannel, default=0}
          Translation.Y {type=MovieSceneScriptingFloatChannel, keys [[0,-1980.0549],[151.347,-1991.8867],[516.064,-1839.474],[528.444,-7.0294]]}
          Angle {type=MovieSceneScriptingFloatChannel, default=0}
          Scale.X {type=MovieSceneScriptingFloatChannel, default=0.75, const [1,1]}
          Scale.Y {type=MovieSceneScriptingFloatChannel, default=0.75, const [1,1]}
          Shear.X {type=MovieSceneScriptingFloatChannel, default=0}
          Shear.Y {type=MovieSceneScriptingFloatChannel, default=0}
         ]
        }]
      }
     ]
    }
    Panel3BG {
     id=#16
     possessed_class=Image
     tracks [{
       type=MovieScene2DTransformTrack
       sections [{
         type=MovieScene2DTransformSection
         end_frame=833
         start_frame=-3
         channels [
          Translation.X {type=MovieSceneScriptingFloatChannel, default=0}
          Translation.Y {type=MovieSceneScriptingFloatChannel, default=0}
          Angle {type=MovieSceneScriptingFloatChannel, default=0}
          Scale.X {type=MovieSceneScriptingFloatChannel, default=1, const [1,1]}
          Scale.Y {type=MovieSceneScriptingFloatChannel, default=1, const [1,1]}
          Shear.X {type=MovieSceneScriptingFloatChannel, default=0}
          Shear.Y {type=MovieSceneScriptingFloatChannel, default=0}
         ]
        }]
      }]
    }
    P3Chr04 {
     id=#17
     possessed_class=Image
     tracks [{
       type=MovieScene2DTransformTrack
       sections [{
         type=MovieScene2DTransformSection
         end_frame=832
         start_frame=-3
         channels [
          Translation.X {type=MovieSceneScriptingFloatChannel, default=0, const [386.5407,1]}
          Translation.Y {type=MovieSceneScriptingFloatChannel, default=0, const [-453.5736,1]}
          Angle {type=MovieSceneScriptingFloatChannel, default=0}
          Scale.X {type=MovieSceneScriptingFloatChannel, keys [[637.498,1.1],[832.93,1.07]]}
          Scale.Y {type=MovieSceneScriptingFloatChannel, keys [[637.498,1.1],[832.93,1.07]]}
          Shear.X {type=MovieSceneScriptingFloatChannel, default=0}
          Shear.Y {type=MovieSceneScriptingFloatChannel, default=0}
         ]
        }]
      }]
    }
    01_ChrB {
     id=#18
     possessed_class=Image
     tracks [
      {
       type=MovieSceneFloatTrack
       display_name=Render Opacity
       sections [{
         type=MovieSceneFloatSection
         end_frame=169
         start_frame=-3
         channels [None {type=MovieSceneScriptingFloatChannel, keys [[-2.421,0],[157,0],[168,1]]}]
        }]
      }
      {
       type=MovieScene2DTransformTrack
       sections [{
         type=MovieScene2DTransformSection
         end_frame=194
         start_frame=-3
         channels [
          Translation.X {type=MovieSceneScriptingFloatChannel, const [102.1673,1]}
          Translation.Y {type=MovieSceneScriptingFloatChannel, const [92.0813,1]}
          Angle {type=MovieSceneScriptingFloatChannel, default=0}
          Scale.X {type=MovieSceneScriptingFloatChannel, const [1.1,1]}
          Scale.Y {type=MovieSceneScriptingFloatChannel, const [1.1,1]}
          Shear.X {type=MovieSceneScriptingFloatChannel, default=0}
          Shear.Y {type=MovieSceneScriptingFloatChannel, default=0}
         ]
        }]
      }
     ]
    }
    Panel2FG {
     id=#19
     possessed_class=Image
     tracks [{
       type=MovieScene2DTransformTrack
       sections [{
         type=MovieScene2DTransformSection
         end_frame=648
         start_frame=404
         channels [
          Translation.X {type=MovieSceneScriptingFloatChannel, default=0}
          Translation.Y {type=MovieSceneScriptingFloatChannel, default=0, const [1080,1]}
          Angle {type=MovieSceneScriptingFloatChannel, default=0}
          Scale.X {type=MovieSceneScriptingFloatChannel, keys [[404.075,1],[648.225,1.03]]}
          Scale.Y {type=MovieSceneScriptingFloatChannel, keys [[404.075,1],[648.225,1.03]]}
          Shear.X {type=MovieSceneScriptingFloatChannel, default=0}
          Shear.Y {type=MovieSceneScriptingFloatChannel, default=0}
         ]
        }]
      }]
    }
    05_Chr {
     id=#20
     possessed_class=Image
     tracks [{
       type=MovieScene2DTransformTrack
       sections [{
         type=MovieScene2DTransformSection
         end_frame=483
         start_frame=-3
         channels [
          Translation.X {type=MovieSceneScriptingFloatChannel, keys [[308.495,-2538.6533],[325.814,803.4614]]}
          Translation.Y {type=MovieSceneScriptingFloatChannel, keys [[308.495,-1561.0176],[325.814,375.9013]]}
          Angle {type=MovieSceneScriptingFloatChannel, default=0}
          Scale.X {type=MovieSceneScriptingFloatChannel, keys [[325.814,1.3],[483.124,1.2]]}
          Scale.Y {type=MovieSceneScriptingFloatChannel, keys [[325.814,1.3],[483.124,1.2]]}
          Shear.X {type=MovieSceneScriptingFloatChannel, default=0}
          Shear.Y {type=MovieSceneScriptingFloatChannel, default=0}
         ]
        }]
      }]
    }
    05_BG {
     id=#21
     possessed_class=Image
     tracks [{
       type=MovieScene2DTransformTrack
       sections [{
         type=MovieScene2DTransformSection
         end_frame=420
         start_frame=-3
         channels [
          Translation.X {type=MovieSceneScriptingFloatChannel, keys [[0,-3643.2712],[223.644,-3643.2712],[223.792,-3908.7385],[232.536,0]]}
          Translation.Y {type=MovieSceneScriptingFloatChannel, keys [[0,2340.7009],[223.644,2340.7009],[223.792,-2181.9753],[232.536,0]]}
         ]
        }]
      }]
    }
    BackgroundGroup2 {
     id=#22
     possessed_class=CanvasPanel
     tracks [{
       type=MovieScene2DTransformTrack
       sections [{
         type=MovieScene2DTransformSection
         end_frame=451
         start_frame=-3
         channels [
          Translation.X {type=MovieSceneScriptingFloatChannel, default=0}
          Translation.Y {type=MovieSceneScriptingFloatChannel, default=0}
          Angle {type=MovieSceneScriptingFloatChannel, default=0}
          Scale.X {type=MovieSceneScriptingFloatChannel, const [1,1]}
          Scale.Y {type=MovieSceneScriptingFloatChannel, const [1,1]}
          Shear.X {type=MovieSceneScriptingFloatChannel, default=0}
          Shear.Y {type=MovieSceneScriptingFloatChannel, default=0}
         ]
        }]
      }]
    }
    Panel2 {
     id=#23
     possessed_class=CanvasPanel
     tracks [
      {
       type=MovieSceneFloatTrack
       display_name=Render Opacity
       sections [{
         type=MovieSceneFloatSection
         end_frame=767
         start_frame=-3
         channels [None {type=MovieSceneScriptingFloatChannel, keys [[0,0],[390.554,1],[658,1],[658.128,0]]}]
        }]
      }
      {
       type=MovieScene2DTransformTrack
       sections [{
         type=MovieScene2DTransformSection
         end_frame=683
         start_frame=0
         channels [
          Translation.X {type=MovieSceneScriptingFloatChannel, thinned_from=4, keys [[0,0],[637.391,0],[658,1920]]}
          Translation.Y {type=MovieSceneScriptingFloatChannel, default=0}
          Angle {type=MovieSceneScriptingFloatChannel, default=0}
          Scale.X {type=MovieSceneScriptingFloatChannel, default=0.5}
          Scale.Y {type=MovieSceneScriptingFloatChannel, default=0.5}
          Shear.X {type=MovieSceneScriptingFloatChannel, default=0}
          Shear.Y {type=MovieSceneScriptingFloatChannel, default=0}
         ]
        }]
      }
     ]
    }
    03_Chr {
     id=#24
     possessed_class=Image
     tracks [
      {
       type=MovieSceneFloatTrack
       display_name=Render Opacity
       sections [{
         type=MovieSceneFloatSection
         end_frame=693
         start_frame=0
         channels [None {type=MovieSceneScriptingFloatChannel, default=1, keys [[0,1],[478,0]]}]
        }]
      }
      {
       type=MovieScene2DTransformTrack
       sections [{
         type=MovieScene2DTransformSection
         end_frame=131
         start_frame=-3
         channels [
          Translation.X {type=MovieSceneScriptingFloatChannel, keys [[75.259,-1115.5704],[95.727,1630.4271]]}
          Translation.Y {type=MovieSceneScriptingFloatChannel, default=2340.7009, keys [[75.259,2379.0037],[95.727,847.178]]}
          Angle {type=MovieSceneScriptingFloatChannel, default=0}
          Scale.X {type=MovieSceneScriptingFloatChannel, default=1, const [1.2,1]}
          Scale.Y {type=MovieSceneScriptingFloatChannel, default=1, const [1.2,1]}
          Shear.X {type=MovieSceneScriptingFloatChannel, default=0}
          Shear.Y {type=MovieSceneScriptingFloatChannel, default=0}
         ]
        }]
      }
     ]
    }
    Panel2Chr6Image {
     id=#25
     possessed_class=Image
     tracks [
      {
       type=MovieSceneFloatTrack
       display_name=Render Opacity
       sections [{
         type=MovieSceneFloatSection
         end_frame=658
         start_frame=-3
         channels [None {type=MovieSceneScriptingFloatChannel, keys [[-2.421,0],[390.554,1],[658,0]]}]
        }]
      }
      {
       type=MovieScene2DTransformTrack
       sections [{
         type=MovieScene2DTransformSection
         end_frame=737
         start_frame=0
         channels [
          Translation.X {type=MovieSceneScriptingFloatChannel, default=0}
          Translation.Y {type=MovieSceneScriptingFloatChannel, keys [[0,-1987.8555],[151.347,-2031.9237],[428.491,-1839.474],[440.156,-7.0294]]}
          Angle {type=MovieSceneScriptingFloatChannel, default=0}
          Scale.X {type=MovieSceneScriptingFloatChannel, default=0.75, const [1,1]}
          Scale.Y {type=MovieSceneScriptingFloatChannel, default=0.75, const [1,1]}
          Shear.X {type=MovieSceneScriptingFloatChannel, default=0}
          Shear.Y {type=MovieSceneScriptingFloatChannel, default=0}
         ]
        }]
      }
     ]
    }
    06_Chr {
     id=#26
     possessed_class=Image
     tracks [{
       type=MovieScene2DTransformTrack
       sections [{
         type=MovieScene2DTransformSection
         end_frame=308
         start_frame=-3
         channels [
          Translation.X {type=MovieSceneScriptingFloatChannel, keys [[278.899,3811.4324],[296.699,1865.142]]}
          Translation.Y {type=MovieSceneScriptingFloatChannel, keys [[278.899,972.8976],[296.699,-193.1739]]}
          Angle {type=MovieSceneScriptingFloatChannel, default=0}
          Scale.X {type=MovieSceneScriptingFloatChannel, default=1, const [0.8,1]}
          Scale.Y {type=MovieSceneScriptingFloatChannel, default=1, const [0.8,1]}
          Shear.X {type=MovieSceneScriptingFloatChannel, default=0}
          Shear.Y {type=MovieSceneScriptingFloatChannel, default=0}
         ]
        }]
      }]
    }
    P3Chr01 {
     id=#27
     possessed_class=Image
     tracks [{
       type=MovieScene2DTransformTrack
       sections [{
         type=MovieScene2DTransformSection
         end_frame=832
         start_frame=636
         channels [
          Translation.X {type=MovieSceneScriptingFloatChannel, default=0, const [47.0592,1]}
          Translation.Y {type=MovieSceneScriptingFloatChannel, default=0, const [-485.2943,1]}
          Angle {type=MovieSceneScriptingFloatChannel, default=0}
          Scale.X {type=MovieSceneScriptingFloatChannel, keys [[637.498,1.5],[832.93,1.55]]}
          Scale.Y {type=MovieSceneScriptingFloatChannel, keys [[637.498,1.5],[832.93,1.55]]}
          Shear.X {type=MovieSceneScriptingFloatChannel, default=0}
          Shear.Y {type=MovieSceneScriptingFloatChannel, default=0}
         ]
        }]
      }]
    }
    03_BG {
     id=#28
     possessed_class=Image
     tracks [
      {
       type=MovieScene2DTransformTrack
       sections [{
         type=MovieScene2DTransformSection
         end_frame=33
         start_frame=-3
         channels [
          Translation.X {type=MovieSceneScriptingFloatChannel, keys [[0,-3643.2712],[14.293,0]]}
          Translation.Y {type=MovieSceneScriptingFloatChannel, keys [[0,2340.7009],[14.293,0]]}
         ]
        }]
      }
      {
       type=MovieSceneFloatTrack
       display_name=Render Opacity
       sections [{
         type=MovieSceneFloatSection
         end_frame=693
         start_frame=0
         channels [None {type=MovieSceneScriptingFloatChannel, default=1, keys [[0,1],[478,0]]}]
        }]
      }
     ]
    }
    P3Chr05 {
     id=#29
     possessed_class=Image
     tracks [{
       type=MovieScene2DTransformTrack
       sections [{
         type=MovieScene2DTransformSection
         end_frame=832
         start_frame=-3
         channels [
          Translation.X {type=MovieSceneScriptingFloatChannel, default=0, const [-813.5319,1]}
          Translation.Y {type=MovieSceneScriptingFloatChannel, default=0, const [-453.3298,1]}
          Angle {type=MovieSceneScriptingFloatChannel, default=0}
          Scale.X {type=MovieSceneScriptingFloatChannel, keys [[637.498,1],[832.93,0.97]]}
          Scale.Y {type=MovieSceneScriptingFloatChannel, keys [[637.498,1],[832.93,0.97]]}
          Shear.X {type=MovieSceneScriptingFloatChannel, default=0}
          Shear.Y {type=MovieSceneScriptingFloatChannel, default=0}
         ]
        }]
      }]
    }
    Panel2Chr2Image {
     id=#30
     possessed_class=Image
     tracks [
      {
       type=MovieSceneFloatTrack
       display_name=Render Opacity
       sections [{
         type=MovieSceneFloatSection
         end_frame=658
         start_frame=-3
         channels [None {type=MovieSceneScriptingFloatChannel, keys [[-2.421,0],[390.554,1],[658,0]]}]
        }]
      }
      {
       type=MovieScene2DTransformTrack
       sections [{
         type=MovieScene2DTransformSection
         end_frame=731
         start_frame=0
         channels [
          Translation.X {type=MovieSceneScriptingFloatChannel, default=0}
          Translation.Y {type=MovieSceneScriptingFloatChannel, keys [[0,-2093.4929],[486.873,-2093.4929],[502.115,-7.0294]]}
          Angle {type=MovieSceneScriptingFloatChannel, default=0}
          Scale.X {type=MovieSceneScriptingFloatChannel, default=0.75, const [1,1]}
          Scale.Y {type=MovieSceneScriptingFloatChannel, default=0.75, const [1,1]}
          Shear.X {type=MovieSceneScriptingFloatChannel, default=0}
          Shear.Y {type=MovieSceneScriptingFloatChannel, default=0}
         ]
        }]
      }
     ]
    }
    04_Chr {
     id=#31
     possessed_class=Image
     tracks [{
       type=MovieScene2DTransformTrack
       sections [{
         type=MovieScene2DTransformSection
         end_frame=251
         start_frame=-3
         channels [
          Translation.X {type=MovieSceneScriptingFloatChannel, keys [[233.432,-3643.2712],[233.58,2685.8877],[251.38,4.3321]]}
          Translation.Y {type=MovieSceneScriptingFloatChannel, keys [[233.432,2340.7009],[233.58,2191.9854],[251.38,565.467]]}
          Angle {type=MovieSceneScriptingFloatChannel, default=0}
          Scale.X {type=MovieSceneScriptingFloatChannel, default=1, const [0.9,1]}
          Scale.Y {type=MovieSceneScriptingFloatChannel, default=1, const [0.9,1]}
          Shear.X {type=MovieSceneScriptingFloatChannel, default=0}
          Shear.Y {type=MovieSceneScriptingFloatChannel, default=0}
         ]
        }]
      }]
    }
    Panel2BG {id=#32, possessed_class=Image, tracks []}
    P3Chr06 {
     id=#33
     possessed_class=Image
     tracks [{
       type=MovieScene2DTransformTrack
       sections [{
         type=MovieScene2DTransformSection
         end_frame=832
         start_frame=-3
         channels [
          Translation.X {type=MovieSceneScriptingFloatChannel, default=0, const [-639.2354,1]}
          Translation.Y {type=MovieSceneScriptingFloatChannel, default=0, const [-449.0716,1]}
          Angle {type=MovieSceneScriptingFloatChannel, default=0}
          Scale.X {type=MovieSceneScriptingFloatChannel, keys [[637.498,1.05],[832.93,1.02]]}
          Scale.Y {type=MovieSceneScriptingFloatChannel, keys [[637.498,1.05],[832.93,1.02]]}
          Shear.X {type=MovieSceneScriptingFloatChannel, default=0}
          Shear.Y {type=MovieSceneScriptingFloatChannel, default=0}
         ]
        }]
      }]
    }
   ]
   display_rate {numerator=60}
  }
 }
 FullWipe {
  binding_count=1
  length_seconds=0.6622
  object=&self:FullWipe
  sequence {bindings [CanvasPanel_59 {
     id=#34
     possessed_class=CanvasPanel
     tracks [
      {
       type=MovieScene2DTransformTrack
       sections [{
         type=MovieScene2DTransformSection
         end_frame=13
         start_frame=0
         channels [
          Translation.X {type=MovieSceneScriptingFloatChannel, keys [[0,0],[13,-3929.5156]]}
          Translation.Y {type=MovieSceneScriptingFloatChannel, const [0,1]}
          Angle {type=MovieSceneScriptingFloatChannel, const [0,1]}
          Scale.X {type=MovieSceneScriptingFloatChannel, default=1}
          Scale.Y {type=MovieSceneScriptingFloatChannel, default=1}
          Shear.X {type=MovieSceneScriptingFloatChannel, default=0}
          Shear.Y {type=MovieSceneScriptingFloatChannel, default=0}
         ]
        }]
      }
      {
       type=MovieSceneFloatTrack
       display_name=Render Opacity
       sections [{type=MovieSceneFloatSection, end_frame=13, start_frame=8, channels [None {type=MovieSceneScriptingFloatChannel, const [1,2]}]}]
      }
     ]
    }]}
 }
]
graphs [EventGraph {nodes [
   n1 = HeroIntroFinished() {is=title K2Node_CustomEvent}
   PlayAnimation(execute=[@n1.then, @n3.then], InAnimation=Get FullWipe() -> (FullWipe:Widget Animation Object Reference))
   > Delay(Duration=1)
   > RemoveFromParent()
   Event PlayPistolCanaryHeroIntroComic() {is=title K2Node_Event}
   n2 = Get ComicHeroIntro() -> (ComicHeroIntro:Widget Animation Object Reference)
   > PlayAnimation(InAnimation=@n2)
   > BindToAnimationFinished(Animation=@n2, Delegate=@n1)
   > PlaySound2D(Sound=/PistolCanary_Event_Audio/Assets/Section02/Heroes/SC_PistolCanary_Section02_Character_Intro_01)
   n3 = Event PlayPistolCanaryIntroWipeAnimation() {is=title K2Node_Event}
   Event ComicIntroUnloadTexturesPanel1() {is=title K2Node_Event}
   > n4 = For Each Loop(Array:Array of Image Object References=Make Array!(
     "[0]":Image Object Reference=Get 01_BG() -> (01_BG:Image Object Reference)
     "[1]":Image Object Reference=Get 01_ChrA() -> (01_ChrA:Image Object Reference)
     "[2]":Image Object Reference=Get 01_ChrB() -> (01_ChrB:Image Object Reference)
     "[3]":Image Object Reference=Get 02_BG() -> (02_BG:Image Object Reference)
     "[4]":Image Object Reference=Get 02_Chr() -> (02_Chr:Image Object Reference)
     "[5]":Image Object Reference=Get 03_BG() -> (03_BG:Image Object Reference)
     "[6]":Image Object Reference=Get 03_Chr() -> (03_Chr:Image Object Reference)
     "[7]":Image Object Reference=Get 04_BG() -> (04_BG:Image Object Reference)
     "[8]":Image Object Reference=Get 04_Chr() -> (04_Chr:Image Object Reference)
     "[9]":Image Object Reference=Get 05_BG() -> (05_BG:Image Object Reference)
     "[10]":Image Object Reference=Get 05_Chr() -> (05_Chr:Image Object Reference)
     "[11]":Image Object Reference=Get 06_BG() -> (06_BG:Image Object Reference)
     "[12]":Image Object Reference=Get 06_Chr() -> (06_Chr:Image Object Reference)
    ) -> (Array:Array of Image Object References)) -> (Array Element:Image Object Reference)
   CollectGarbage!(execute=@n4.Completed) -> (then)
   Is Valid(exec=@n4.LoopBody, InputObject=@n4)
   > SetBrushFromTexture!(execute, self:Image Object Reference=@n4, Texture:Texture 2D Object Reference, bMatchSize:bool=true) -> (then)
   Event ComicIntroUnloadTexturesPanel2() {is=title K2Node_Event}
   > n5 = For Each Loop(Array=Make Array!(
     "[0]"=Get Panel2FG() -> (Panel2FG:Image Object Reference)
     "[1]"=Get Panel2Chr6Image() -> (Panel2Chr6Image:Image Object Reference)
     "[2]"=Get Panel2Chr5Image() -> (Panel2Chr5Image:Image Object Reference)
     "[3]"=Get Panel2Chr4Image() -> (Panel2Chr4Image:Image Object Reference)
     "[4]"=Get Panel2Chr3Image() -> (Panel2Chr3Image:Image Object Reference)
     "[5]"=Get Panel2Chr2Image() -> (Panel2Chr2Image:Image Object Reference)
     "[6]"=Get Panel2Chr1Image() -> (Panel2Chr1Image:Image Object Reference)
     "[7]"=Get Panel2BG() -> (Panel2BG:Image Object Reference)
    ) -> (Array))
   CollectGarbage(execute=@n5.Completed)
   Is Valid(exec=@n5.LoopBody, InputObject=@n5)
   > SetBrushFromTexture(self=@n5, bMatchSize=true)
   Event ComicIntroUnloadTexturesPanel3() {is=title K2Node_Event}
   > n6 = For Each Loop(Array=Make Array!(
     "[0]"=Get Panel3BG() -> (Panel3BG:Image Object Reference)
     "[1]"=Get P3Chr06() -> (P3Chr06:Image Object Reference)
     "[2]"=Get P3Chr05() -> (P3Chr05:Image Object Reference)
     "[3]"=Get P3Chr04() -> (P3Chr04:Image Object Reference)
     "[4]"=Get P3Chr03() -> (P3Chr03:Image Object Reference)
     "[5]"=Get P3Chr02() -> (P3Chr02:Image Object Reference)
     "[6]"=Get P3Chr01() -> (P3Chr01:Image Object Reference)
    ) -> (Array))
   CollectGarbage(execute=@n6.Completed)
   Is Valid(exec=@n6.LoopBody, InputObject=@n6)
   > SetBrushFromTexture(self=@n6, bMatchSize=true)
   Event ComicIntroLoadTextures() {is=title K2Node_Event}
   > PrintString!(execute, InString=Loading Comic Intro textures, bPrintToScreen, bPrintToLog, TextColor, Duration, Key) -> (then)
   > n7 = For Each Loop(Array=Make Array!("[0]"=Get 01_BG(), "[1]"=Get 02_BG(), "[2]"=Get 03_BG(), "[3]"=Get 04_BG(), "[4]"=Get 05_BG(), "[5]"=Get 06_BG()) -> (Array))
   n8 = For Each Loop(
    Exec=@n7.Completed
    Array=Make Array!(
     "[0]"=Get 01_ChrA()
     "[1]"=Get 01_ChrB()
     "[2]"=Get 02_Chr()
     "[3]"=Get 03_Chr()
     "[4]"=Get 04_Chr()
     "[5]"=Get 05_Chr()
     "[6]":Image Object Reference=Get 06_Chr()
    ) -> (Array)
   )
   Delay(execute=@n8.Completed, Duration=2)
   > n9 = For Each Loop(Array=Make Array!(
     "[0]"=Get Panel2BG()
     "[1]"=Get Panel2FG()
     "[2]"=Get Panel2Chr1Image()
     "[3]"=Get Panel2Chr2Image()
     "[4]"=Get Panel2Chr3Image()
     "[5]"=Get Panel2Chr4Image()
     "[6]"=Get Panel2Chr5Image()
     "[7]":Image Object Reference=Get Panel2Chr6Image()
    ) -> (Array))
   Delay(execute=@n9.Completed, Duration=2)
   > n10 = For Each Loop(Array=Make Array!(
     "[0]"=Get Panel3BG()
     "[1]"=Get P3Chr01()
     "[2]"=Get P3Chr02()
     "[3]"=Get P3Chr03()
     "[4]"=Get P3Chr04()
     "[5]"=Get P3Chr05()
     "[6]"=Get P3Chr06()
    ) -> (Array))
   n11 = IsMobilePlatform!() -> (ReturnValue:bool)
   n12 = Get Panel3_Textures() -> (Panel3_Textures:Array of Texture 2D Soft Object References)
   n13 = Get Panel3_TexturesLow() -> (Panel3_TexturesLow:Array of Texture 2D Soft Object References)
   > SetBrushFromSoftTexture!(
    execute
    self:Image Object Reference=@n10
    SoftTexture:Texture 2D Soft Object Reference="Get (a copy)"(
     Array:Array of Texture 2D Soft Object References=Select(Option 0:Array of Texture 2D Soft Object References=@n12, Option 1:Array of Texture 2D Soft Object References=@n13, Index=@n11) -> (ReturnValue:Array of Texture 2D Soft Object References)
     Dimension 1=@n10.Array Index
    ) -> (Output:Texture 2D Soft Object Reference)
    bMatchSize:bool=false
   ) -> (then)
   n14 = IsMobilePlatform()
   n15 = Get Panel2_Textures() -> (Panel2_Textures:Array of Texture 2D Soft Object References)
   n16 = Get Panel2_TexturesLow() -> (Panel2_TexturesLow:Array of Texture 2D Soft Object References)
   SetBrushFromSoftTexture(
    execute=@n9.LoopBody
    self=@n9
    SoftTexture="Get (a copy)"(Array=Select(Option 0=@n15, Option 1=@n16, Index=@n14), Dimension 1=@n9.Array Index)
    bMatchSize=false
   )
   n17 = IsMobilePlatform()
   n18 = Get Panel1_TexturesLow() -> (Panel1_TexturesLow:Array of Texture 2D Soft Object References)
   n19 = Get Panel1_Textures() -> (Panel1_Textures:Array of Texture 2D Soft Object References)
   SetBrushFromSoftTexture(
    execute=@n8.LoopBody
    self=@n8
    SoftTexture="Get (a copy)"(Array=Select(Option 0=@n19, Option 1=@n18, Index=@n17), Dimension 1=@n8.Array Index)
    bMatchSize=false
   )
   SetBrushFromSoftMaterial!(
    execute=@n7.LoopBody
    self:Image Object Reference=@n7
    SoftMaterial:Material Interface Soft Object Reference="Get (a copy)"(
     Array:Array of Material Interface Soft Object References=Get Panel1_Materials() -> (Panel1_Materials:Array of Material Interface Soft Object References)
     Dimension 1=@n7.Array Index
    ) -> (Output:Material Interface Soft Object Reference)
   ) -> (then)
   Event PreConstruct()
   Event Construct()
   Event Tick()
  ]}]
variables [
 Panel1_Materials {
  type="array<MaterialInterface>"
  default [
   /PistolCanary/Layout/Material/Boards/MI_PanelBG_Panning_KillSwitch
   /PistolCanary/Layout/Material/Boards/MI_PanelBG_Panning_Raven
   /PistolCanary/Layout/Material/Boards/MI_PanelBG_Panning_LightRider
   /PistolCanary/Layout/Material/Boards/MI_PanelBG_Panning_SynthWave
   /PistolCanary/Layout/Material/Boards/MI_PanelBG_Panning_StormBeast
   /PistolCanary/Layout/Material/Boards/MI_PanelBG_Panning_Ziggy
  ]
 }
 Panel1_Textures {
  type="array<Texture2D>"
  default [
   /PistolCanary/Layout/Texture/ComicPanels/IntroPanel1/PC_Color_Panel_01_Panel_1_Character
   /PistolCanary/Layout/Texture/ComicPanels/IntroPanel1/PC_Color_01_panel_1_visor
   /PistolCanary/Layout/Texture/ComicPanels/IntroPanel1/PC_Color_Panel_01_Panel_2_Character
   /PistolCanary/Layout/Texture/ComicPanels/IntroPanel1/PC_Color_Panel_01_Panel_3_Character
   /PistolCanary/Layout/Texture/ComicPanels/IntroPanel1/PC_Color_Panel_02_Panel_1_Character
   /PistolCanary/Layout/Texture/ComicPanels/IntroPanel1/PC_Color_Panel_02_Panel_2_Character
   /PistolCanary/Layout/Texture/ComicPanels/IntroPanel1/PC_Color_Panel_02_Panel_3_Character
  ]
 }
 Panel1_TexturesLow {
  type="array<Texture2D>"
  default [
   /PistolCanary/Layout/Texture/ComicPanels/IntroPanel1/LowRes/PC_Color_Panel_01_Panel_1_Character_no_mask_low
   /PistolCanary/Layout/Texture/ComicPanels/IntroPanel1/LowRes/PC_Color_01_panel_1_visor_low
   /PistolCanary/Layout/Texture/ComicPanels/IntroPanel1/LowRes/PC_Color_Panel_01_Panel_2_Character_low
   /PistolCanary/Layout/Texture/ComicPanels/IntroPanel1/LowRes/PC_Color_Panel_01_Panel_3_Character_low
   /PistolCanary/Layout/Texture/ComicPanels/IntroPanel1/LowRes/PC_Color_Panel_02_Panel_1_Character_low
   /PistolCanary/Layout/Texture/ComicPanels/IntroPanel1/LowRes/PC_Color_Panel_02_Panel_2_Character_low
   /PistolCanary/Layout/Texture/ComicPanels/IntroPanel1/LowRes/PC_Color_Panel_02_Panel_3_Character_low
  ]
 }
 Panel2_Textures {
  type="array<Texture2D>"
  default [
   /PistolCanary/Layout/Texture/ComicPanels/IntroPanel2/PC_Color_03_Background
   /PistolCanary/Layout/Texture/ComicPanels/IntroPanel2/PC_Color_03_Foreground
   /PistolCanary/Layout/Texture/ComicPanels/IntroPanel2/Characters/PC_Colors_03a
   /PistolCanary/Layout/Texture/ComicPanels/IntroPanel2/Characters/PC_Colors_03b
   /PistolCanary/Layout/Texture/ComicPanels/IntroPanel2/Characters/PC_Colors_03c
   /PistolCanary/Layout/Texture/ComicPanels/IntroPanel2/Characters/PC_Colors_03d
   /PistolCanary/Layout/Texture/ComicPanels/IntroPanel2/Characters/PC_Colors_03e
   /PistolCanary/Layout/Texture/ComicPanels/IntroPanel2/Characters/PC_Colors_03f
  ]
 }
 Panel2_TexturesLow {
  type="array<Texture2D>"
  default [
   /PistolCanary/Layout/Texture/ComicPanels/IntroPanel2/LowRes/PC_Color_03_Background_low
   /PistolCanary/Layout/Texture/ComicPanels/IntroPanel2/LowRes/PC_Color_03_Foreground_low
   /PistolCanary/Layout/Texture/ComicPanels/IntroPanel2/LowRes/PC_Colors_03a_low
   /PistolCanary/Layout/Texture/ComicPanels/IntroPanel2/LowRes/PC_Colors_03b_low
   /PistolCanary/Layout/Texture/ComicPanels/IntroPanel2/LowRes/PC_Colors_03c_low
   /PistolCanary/Layout/Texture/ComicPanels/IntroPanel2/LowRes/PC_Colors_03d_low
   /PistolCanary/Layout/Texture/ComicPanels/IntroPanel2/LowRes/PC_Colors_03e_low
   /PistolCanary/Layout/Texture/ComicPanels/IntroPanel2/LowRes/PC_Colors_03f_low
  ]
 }
 Panel3_Textures {
  type="array<Texture2D>"
  default [
   /PistolCanary/Layout/Texture/ComicPanels/IntroPanel3/Pistol_Canary_colors_04_layers_background2
   /PistolCanary/Layout/Texture/ComicPanels/IntroPanel3/PC_Color_04a
   /PistolCanary/Layout/Texture/ComicPanels/IntroPanel3/PC_Color_04b
   /PistolCanary/Layout/Texture/ComicPanels/IntroPanel3/PC_Color_04c
   /PistolCanary/Layout/Texture/ComicPanels/IntroPanel3/PC_Color_04d
   /PistolCanary/Layout/Texture/ComicPanels/IntroPanel3/PC_Color_04e
   /PistolCanary/Layout/Texture/ComicPanels/IntroPanel3/PC_Color_04f
  ]
 }
 Panel3_TexturesLow {
  type="array<Texture2D>"
  default [
   /PistolCanary/Layout/Texture/ComicPanels/IntroPanel3/LowRes/Pistol_Canary_colors_04_layers_background2_low
   /PistolCanary/Layout/Texture/ComicPanels/IntroPanel3/LowRes/PC_Color_04a_low
   /PistolCanary/Layout/Texture/ComicPanels/IntroPanel3/LowRes/PC_Color_04b_low
   /PistolCanary/Layout/Texture/ComicPanels/IntroPanel3/LowRes/PC_Color_04c_low
   /PistolCanary/Layout/Texture/ComicPanels/IntroPanel3/LowRes/PC_Color_04d_low
   /PistolCanary/Layout/Texture/ComicPanels/IntroPanel3/LowRes/PC_Color_04e_low
   /PistolCanary/Layout/Texture/ComicPanels/IntroPanel3/LowRes/PC_Color_04f_low
  ]
 }
]
widgets {
 info {inheritedWidgetCount=0, parentClass=&UserWidget, rootWidgetClass=&CanvasPanel, widgetCount=36}
 widgets [
  CanvasPanel_59 {widgetClassPath=&CanvasPanel, widget=CanvasPanel_59, properties {clipping=ClipToBounds, visibility=HitTestInvisible}}
  BG_Left {
   widgetClassPath=&Image
   bIsVariable=true
   parent=CanvasPanel_59
   widget=BG_Left
   properties {
    visibility=HitTestInvisible
    brush {
     bIsDynamicallyLoaded=false
     drawAs=Box
     imageType=NoImage
     mirroring=NoMirror
     resourceName=None
     resourceObject=null
     tiling=NoTile
     imageSize=<32 32>
     margin=<ltrb 0 0 0 0>
     outlineSettings {
      bUseBrushTransparency=false
      roundingType=HalfHeightRadius
      width=0
      color {colorUseRule=UseColor_Specified, specifiedColor=<rgba 0 0 0 0>}
      cornerRadii=<0 0 0 0>
     }
     tintColor {colorUseRule=UseColor_Specified, specifiedColor=<rgba 1 1 1 1>}
     uVRegion {bIsValid=false, max=<0 0>, min=<0 0>}
    }
    colorAndOpacity=<rgba 0 0 0 1>
   }
   slot {
    class=CanvasPanelSlot
    properties {layoutData {alignment=<0 0>, anchors {maximum=<0 0.5>, minimum=<0 0.5>}, offsets=<ltrb -1920 -540 1920 1080>}}
   }
  }
  BG_Right {
   widgetClassPath=&Image
   bIsVariable=true
   parent=CanvasPanel_59
   widget=BG_Right
   properties {
    visibility=HitTestInvisible
    brush {
     bIsDynamicallyLoaded=false
     drawAs=Box
     imageType=NoImage
     mirroring=NoMirror
     resourceName=None
     resourceObject=null
     tiling=NoTile
     imageSize=<32 32>
     margin=<ltrb 0 0 0 0>
     outlineSettings {
      bUseBrushTransparency=false
      roundingType=HalfHeightRadius
      width=0
      color {colorUseRule=UseColor_Specified, specifiedColor=<rgba 0 0 0 0>}
      cornerRadii=<0 0 0 0>
     }
     tintColor {colorUseRule=UseColor_Specified, specifiedColor=<rgba 1 1 1 1>}
     uVRegion {bIsValid=false, max=<0 0>, min=<0 0>}
    }
    colorAndOpacity=<rgba 0 0 0 1>
   }
   slot {
    class=CanvasPanelSlot
    properties {layoutData {alignment=<0 0>, anchors {maximum=<1 0.5>, minimum=<1 0.5>}, offsets=<ltrb 0 -540 1920 1080>}}
   }
  }
  Panel1 {
   widgetClassPath=&CanvasPanel
   parent=CanvasPanel_59
   widget=Panel1
   properties {clipping=ClipToBounds, visibility=HitTestInvisible, renderTransform {angle=0, scale=<0.5 0.5>, shear=<0 0>, translation=<0 0>}}
   slot {
    class=CanvasPanelSlot
    properties {zOrder=3, layoutData {alignment=<0 0>, anchors {maximum=<0.5 0.5>, minimum=<0.5 0.5>}, offsets=<ltrb -1920 -1080 3840 2160>}}
   }
  }
  BackgroundGroup1 {
   widgetClassPath=&CanvasPanel
   parent=Panel1
   widget=BackgroundGroup1
   properties {clipping=ClipToBounds, visibility=HitTestInvisible}
   slot {class=CanvasPanelSlot, properties {layoutData {alignment=<0 0>, anchors {maximum=<0 0>, minimum=<0 0>}, offsets=<ltrb 0 0 3840 2160>}}}
  }
  01_BG {
   widgetClassPath=&Image
   bIsVariable=true
   parent=BackgroundGroup1
   widget=01_BG
   properties {
    clipping=ClipToBounds
    pixelSnapping=Disabled
    visibility=HitTestInvisible
    brush {
     bIsDynamicallyLoaded=false
     drawAs=Image
     imageType=FullColor
     mirroring=NoMirror
     resourceName=None
     resourceObject=null
     tiling=NoTile
     imageSize=<1920 1080>
     margin=<ltrb 0 0 0 0>
     outlineSettings {
      bUseBrushTransparency=false
      roundingType=HalfHeightRadius
      width=0
      color {colorUseRule=UseColor_Specified, specifiedColor=<rgba 0 0 0 0>}
      cornerRadii=<0 0 0 0>
     }
     tintColor {colorUseRule=UseColor_Specified, specifiedColor=<rgba 1 1 1 1>}
     uVRegion {bIsValid=false, max=<0 0>, min=<0 0>}
    }
    renderTransform {angle=-121, scale=<1 1>, shear=<0 0>, translation=<0 0>}
   }
   slot {
    class=CanvasPanelSlot
    properties {layoutData {alignment=<0 0>, anchors {maximum=<0 0>, minimum=<0 0>}, offsets=<ltrb 0 -2220.7402 1348.7668 5058.291>}}
   }
  }
  02_BG {
   widgetClassPath=&Image
   bIsVariable=true
   parent=BackgroundGroup1
   widget=02_BG
   properties {
    clipping=ClipToBounds
    pixelSnapping=Disabled
    visibility=HitTestInvisible
    brush {
     bIsDynamicallyLoaded=false
     drawAs=Image
     imageType=FullColor
     mirroring=NoMirror
     resourceName=None
     resourceObject=null
     tiling=NoTile
     imageSize=<1920 1080>
     margin=<ltrb 0 0 0 0>
     outlineSettings {
      bUseBrushTransparency=false
      roundingType=HalfHeightRadius
      width=0
      color {colorUseRule=UseColor_Specified, specifiedColor=<rgba 0 0 0 0>}
      cornerRadii=<0 0 0 0>
     }
     tintColor {colorUseRule=UseColor_Specified, specifiedColor=<rgba 1 1 1 1>}
     uVRegion {bIsValid=false, max=<0 0>, min=<0 0>}
    }
    renderTransform {angle=-121, scale=<1 1>, shear=<0 0>, translation=<4272 -2513>}
   }
   slot {
    class=CanvasPanelSlot
    properties {layoutData {alignment=<0 0>, anchors {maximum=<0 0>, minimum=<0 0>}, offsets=<ltrb 1212.3036 -1464.65 1348.7668 5058.291>}}
   }
  }
  03_BG {
   widgetClassPath=&Image
   bIsVariable=true
   parent=BackgroundGroup1
   widget=03_BG
   properties {
    clipping=ClipToBounds
    pixelSnapping=Disabled
    visibility=HitTestInvisible
    brush {
     bIsDynamicallyLoaded=false
     drawAs=Image
     imageType=FullColor
     mirroring=NoMirror
     resourceName=None
     resourceObject=null
     tiling=NoTile
     imageSize=<3840 2160>
     margin=<ltrb 0 0 0 0>
     outlineSettings {
      bUseBrushTransparency=false
      roundingType=HalfHeightRadius
      width=0
      color {colorUseRule=UseColor_Specified, specifiedColor=<rgba 0 0 0 0>}
      cornerRadii=<0 0 0 0>
     }
     tintColor {colorUseRule=UseColor_Specified, specifiedColor=<rgba 1 1 1 1>}
     uVRegion {bIsValid=false, max=<0 0>, min=<0 0>}
    }
    renderTransform {angle=-120, scale=<1 1>, shear=<0 0>, translation=<0 0>}
   }
   slot {
    class=CanvasPanelSlot
    properties {layoutData {alignment=<0 0>, anchors {maximum=<0 0>, minimum=<0 0>}, offsets=<ltrb 2162.1245 -495.7184 1348.7668 5058.291>}}
   }
  }
  03_Chr {
   widgetClassPath=&Image
   bIsVariable=true
   parent=Panel1
   widget=03_Chr
   properties {
    clipping=ClipToBounds
    pixelSnapping=Disabled
    visibility=HitTestInvisible
    brush {
     bIsDynamicallyLoaded=false
     drawAs=Image
     imageType=FullColor
     mirroring=NoMirror
     resourceName=None
     resourceObject=null
     tiling=NoTile
     imageSize=<800 488>
     margin=<ltrb 0 0 0 0>
     outlineSettings {
      bUseBrushTransparency=false
      roundingType=HalfHeightRadius
      width=0
      color {colorUseRule=UseColor_Specified, specifiedColor=<rgba 0 0 0 0>}
      cornerRadii=<0 0 0 0>
     }
     tintColor {colorUseRule=UseColor_Specified, specifiedColor=<rgba 1 1 1 1>}
     uVRegion {bIsValid=false, max=<0 0>, min=<0 0>}
    }
   }
   slot {class=CanvasPanelSlot, properties {layoutData {alignment=<0 0>, anchors {maximum=<0 0>, minimum=<0 0>}, offsets=<ltrb 0 0 2048 1248>}}}
  }
  02_Chr {
   widgetClassPath=&Image
   bIsVariable=true
   parent=Panel1
   widget=02_Chr
   properties {
    clipping=ClipToBounds
    pixelSnapping=Disabled
    visibility=HitTestInvisible
    brush {
     bIsDynamicallyLoaded=false
     drawAs=Image
     imageType=FullColor
     mirroring=NoMirror
     resourceName=None
     resourceObject=null
     tiling=NoTile
     imageSize=<800 532>
     margin=<ltrb 0 0 0 0>
     outlineSettings {
      bUseBrushTransparency=false
      roundingType=HalfHeightRadius
      width=0
      color {colorUseRule=UseColor_Specified, specifiedColor=<rgba 0 0 0 0>}
      cornerRadii=<0 0 0 0>
     }
     tintColor {colorUseRule=UseColor_Specified, specifiedColor=<rgba 1 1 1 1>}
     uVRegion {bIsValid=false, max=<0 0>, min=<0 0>}
    }
   }
   slot {class=CanvasPanelSlot, properties {layoutData {alignment=<0 0>, anchors {maximum=<0 0>, minimum=<0 0>}, offsets=<ltrb 0 0 2048 1352>}}}
  }
  01_ChrA {
   widgetClassPath=&Image
   bIsVariable=true
   parent=Panel1
   widget=01_ChrA
   properties {
    clipping=ClipToBounds
    pixelSnapping=Disabled
    visibility=HitTestInvisible
    brush {
     bIsDynamicallyLoaded=false
     drawAs=Image
     imageType=FullColor
     mirroring=NoMirror
     resourceName=None
     resourceObject=null
     tiling=NoTile
     imageSize=<800 472>
     margin=<ltrb 0 0 0 0>
     outlineSettings {
      bUseBrushTransparency=false
      roundingType=HalfHeightRadius
      width=0
      color {colorUseRule=UseColor_Specified, specifiedColor=<rgba 0 0 0 0>}
      cornerRadii=<0 0 0 0>
     }
     tintColor {colorUseRule=UseColor_Specified, specifiedColor=<rgba 1 1 1 1>}
     uVRegion {bIsValid=false, max=<0 0>, min=<0 0>}
    }
   }
   slot {class=CanvasPanelSlot, properties {layoutData {alignment=<0 0>, anchors {maximum=<0 0>, minimum=<0 0>}, offsets=<ltrb 0 0 2048 1200>}}}
  }
  01_ChrB {
   widgetClassPath=&Image
   bIsVariable=true
   parent=Panel1
   widget=01_ChrB
   properties {
    clipping=ClipToBounds
    pixelSnapping=Disabled
    visibility=HitTestInvisible
    brush {
     bIsDynamicallyLoaded=false
     drawAs=Image
     imageType=FullColor
     mirroring=NoMirror
     resourceName=None
     resourceObject=null
     tiling=NoTile
     imageSize=<256 160>
     margin=<ltrb 0 0 0 0>
     outlineSettings {
      bUseBrushTransparency=false
      roundingType=HalfHeightRadius
      width=0
      color {colorUseRule=UseColor_Specified, specifiedColor=<rgba 0 0 0 0>}
      cornerRadii=<0 0 0 0>
     }
     tintColor {colorUseRule=UseColor_Specified, specifiedColor=<rgba 1 1 1 1>}
     uVRegion {bIsValid=false, max=<0 0>, min=<0 0>}
    }
    renderTransform {angle=0, scale=<1.3 1.3>, shear=<0 0>, translation=<102.1673 92.0813>}
   }
   slot {
    class=CanvasPanelSlot
    properties {layoutData {alignment=<0 0>, anchors {maximum=<0 0>, minimum=<0 0>}, offsets=<ltrb 739.8011 99.4623 256 160>}}
   }
  }
  BackgroundGroup2 {
   widgetClassPath=&CanvasPanel
   parent=Panel1
   widget=BackgroundGroup2
   properties {clipping=ClipToBounds, visibility=HitTestInvisible}
   slot {
    class=CanvasPanelSlot
    properties {zOrder=1, layoutData {alignment=<0 0>, anchors {maximum=<0 0>, minimum=<0 0>}, offsets=<ltrb 0 0 3840 2160>}}
   }
  }
  06_BG {
   widgetClassPath=&Image
   bIsVariable=true
   parent=BackgroundGroup2
   widget=06_BG
   properties {
    clipping=ClipToBounds
    pixelSnapping=Disabled
    visibility=HitTestInvisible
    brush {
     bIsDynamicallyLoaded=false
     drawAs=Image
     imageType=FullColor
     mirroring=NoMirror
     resourceName=None
     resourceObject=null
     tiling=NoTile
     imageSize=<3840 2160>
     margin=<ltrb 0 0 0 0>
     outlineSettings {
      bUseBrushTransparency=false
      roundingType=HalfHeightRadius
      width=0
      color {colorUseRule=UseColor_Specified, specifiedColor=<rgba 0 0 0 0>}
      cornerRadii=<0 0 0 0>
     }
     tintColor {colorUseRule=UseColor_Specified, specifiedColor=<rgba 1 1 1 1>}
     uVRegion {bIsValid=false, max=<0 0>, min=<0 0>}
    }
    renderTransform {angle=120, scale=<1 1>, shear=<0 0>, translation=<0 0>}
   }
   slot {
    class=CanvasPanelSlot
    properties {layoutData {alignment=<0 0>, anchors {maximum=<0 0>, minimum=<0 0>}, offsets=<ltrb 2630.9819 -2220.8479 1348.7668 5097.126>}}
   }
  }
  04_BG {
   widgetClassPath=&Image
   bIsVariable=true
   parent=BackgroundGroup2
   widget=04_BG
   properties {
    clipping=ClipToBounds
    pixelSnapping=Disabled
    visibility=HitTestInvisible
    brush {
     bIsDynamicallyLoaded=false
     drawAs=Image
     imageType=FullColor
     mirroring=NoMirror
     resourceName=None
     resourceObject=null
     tiling=NoTile
     imageSize=<1920 1080>
     margin=<ltrb 0 0 0 0>
     outlineSettings {
      bUseBrushTransparency=false
      roundingType=HalfHeightRadius
      width=0
      color {colorUseRule=UseColor_Specified, specifiedColor=<rgba 0 0 0 0>}
      cornerRadii=<0 0 0 0>
     }
     tintColor {colorUseRule=UseColor_Specified, specifiedColor=<rgba 1 1 1 1>}
     uVRegion {bIsValid=false, max=<0 0>, min=<0 0>}
    }
    renderTransform {angle=120, scale=<1 1>, shear=<0 0>, translation=<0 0>}
   }
   slot {
    class=CanvasPanelSlot
    properties {layoutData {alignment=<0 0>, anchors {maximum=<0 0>, minimum=<0 0>}, offsets=<ltrb 0 -630.5384 1348.7668 5056.3491>}}
   }
  }
  05_BG {
   widgetClassPath=&Image
   bIsVariable=true
   parent=BackgroundGroup2
   widget=05_BG
   properties {
    clipping=ClipToBounds
    pixelSnapping=Disabled
    visibility=HitTestInvisible
    brush {
     bIsDynamicallyLoaded=false
     drawAs=Image
     imageType=FullColor
     mirroring=NoMirror
     resourceName=None
     resourceObject=null
     tiling=NoTile
     imageSize=<3840 2160>
     margin=<ltrb 0 0 0 0>
     outlineSettings {
      bUseBrushTransparency=false
      roundingType=HalfHeightRadius
      width=0
      color {colorUseRule=UseColor_Specified, specifiedColor=<rgba 0 0 0 0>}
      cornerRadii=<0 0 0 0>
     }
     tintColor {colorUseRule=UseColor_Specified, specifiedColor=<rgba 1 1 1 1>}
     uVRegion {bIsValid=false, max=<0 0>, min=<0 0>}
    }
    renderTransform {angle=120, scale=<1 1>, shear=<0 0>, translation=<0 0>}
   }
   slot {
    class=CanvasPanelSlot
    properties {layoutData {alignment=<0 0>, anchors {maximum=<0 0>, minimum=<0 0>}, offsets=<ltrb 1260.1565 -1443.4039 1348.7668 5058.291>}}
   }
  }
  05_Chr {
   widgetClassPath=&Image
   bIsVariable=true
   parent=Panel1
   widget=05_Chr
   properties {
    clipping=ClipToBounds
    pixelSnapping=Disabled
    visibility=HitTestInvisible
    brush {
     bIsDynamicallyLoaded=false
     drawAs=Image
     imageType=FullColor
     mirroring=NoMirror
     resourceName=None
     resourceObject=null
     tiling=NoTile
     imageSize=<800 552>
     margin=<ltrb 0 0 0 0>
     outlineSettings {
      bUseBrushTransparency=false
      roundingType=HalfHeightRadius
      width=0
      color {colorUseRule=UseColor_Specified, specifiedColor=<rgba 0 0 0 0>}
      cornerRadii=<0 0 0 0>
     }
     tintColor {colorUseRule=UseColor_Specified, specifiedColor=<rgba 1 1 1 1>}
     uVRegion {bIsValid=false, max=<0 0>, min=<0 0>}
    }
   }
   slot {
    class=CanvasPanelSlot
    properties {zOrder=2, layoutData {alignment=<0 0>, anchors {maximum=<0 0>, minimum=<0 0>}, offsets=<ltrb 0 0 2048 1408>}}
   }
  }
  06_Chr {
   widgetClassPath=&Image
   bIsVariable=true
   parent=Panel1
   widget=06_Chr
   properties {
    clipping=ClipToBounds
    pixelSnapping=Disabled
    visibility=HitTestInvisible
    brush {
     bIsDynamicallyLoaded=false
     drawAs=Image
     imageType=FullColor
     mirroring=NoMirror
     resourceName=None
     resourceObject=null
     tiling=NoTile
     imageSize=<800 764>
     margin=<ltrb 0 0 0 0>
     outlineSettings {
      bUseBrushTransparency=false
      roundingType=HalfHeightRadius
      width=0
      color {colorUseRule=UseColor_Specified, specifiedColor=<rgba 0 0 0 0>}
      cornerRadii=<0 0 0 0>
     }
     tintColor {colorUseRule=UseColor_Specified, specifiedColor=<rgba 1 1 1 1>}
     uVRegion {bIsValid=false, max=<0 0>, min=<0 0>}
    }
   }
   slot {
    class=CanvasPanelSlot
    properties {zOrder=2, layoutData {alignment=<0 0>, anchors {maximum=<0 0>, minimum=<0 0>}, offsets=<ltrb 0 0 2048 1952>}}
   }
  }
  04_Chr {
   widgetClassPath=&Image
   bIsVariable=true
   parent=Panel1
   widget=04_Chr
   properties {
    clipping=ClipToBounds
    pixelSnapping=Disabled
    visibility=HitTestInvisible
    brush {
     bIsDynamicallyLoaded=false
     drawAs=Image
     imageType=FullColor
     mirroring=NoMirror
     resourceName=None
     resourceObject=null
     tiling=NoTile
     imageSize=<800 672>
     margin=<ltrb 0 0 0 0>
     outlineSettings {
      bUseBrushTransparency=false
      roundingType=HalfHeightRadius
      width=0
      color {colorUseRule=UseColor_Specified, specifiedColor=<rgba 0 0 0 0>}
      cornerRadii=<0 0 0 0>
     }
     tintColor {colorUseRule=UseColor_Specified, specifiedColor=<rgba 1 1 1 1>}
     uVRegion {bIsValid=false, max=<0 0>, min=<0 0>}
    }
   }
   slot {
    class=CanvasPanelSlot
    properties {zOrder=2, layoutData {alignment=<0 0>, anchors {maximum=<0 0>, minimum=<0 0>}, offsets=<ltrb 0 0 2048 1712>}}
   }
  }
  Panel2 {
   widgetClassPath=&CanvasPanel
   parent=CanvasPanel_59
   widget=Panel2
   properties {
    clipping=ClipToBounds
    visibility=HitTestInvisible
    renderTransform {angle=0, scale=<0.5 0.5>, shear=<0 0>, translation=<0 0>}
    renderTransformPivot=<0 0>
   }
   slot {
    class=CanvasPanelSlot
    properties {zOrder=2, layoutData {alignment=<0 0>, anchors {maximum=<0.5 0.5>, minimum=<0.5 0.5>}, offsets=<ltrb -960 -540 3840 2160>}}
   }
  }
  Panel2BG {
   widgetClassPath=&Image
   bIsVariable=true
   parent=Panel2
   widget=Panel2BG
   properties {
    clipping=ClipToBounds
    visibility=HitTestInvisible
    brush {
     bIsDynamicallyLoaded=false
     drawAs=Box
     imageType=FullColor
     mirroring=NoMirror
     resourceName=None
     resourceObject=null
     tiling=NoTile
     imageSize=<3840 2160>
     margin=<ltrb 0 0 0 0>
     outlineSettings {
      bUseBrushTransparency=false
      roundingType=HalfHeightRadius
      width=0
      color {colorUseRule=UseColor_Specified, specifiedColor=<rgba 0 0 0 0>}
      cornerRadii=<0 0 0 0>
     }
     tintColor {colorUseRule=UseColor_Specified, specifiedColor=<rgba 1 1 1 1>}
     uVRegion {bIsValid=false, max=<0 0>, min=<0 0>}
    }
   }
   slot {class=CanvasPanelSlot, properties {layoutData {alignment=<0 0>, anchors {maximum=<0 0>, minimum=<0 0>}, offsets=<ltrb 0 0 3840 2160>}}}
  }
  Panel2FG {
   widgetClassPath=&Image
   bIsVariable=true
   parent=Panel2
   widget=Panel2FG
   properties {
    clipping=ClipToBounds
    visibility=HitTestInvisible
    brush {
     bIsDynamicallyLoaded=false
     drawAs=Box
     imageType=FullColor
     mirroring=NoMirror
     resourceName=None
     resourceObject=null
     tiling=NoTile
     imageSize=<3840 1080>
     margin=<ltrb 0 0 0 0>
     outlineSettings {
      bUseBrushTransparency=false
      roundingType=HalfHeightRadius
      width=0
      color {colorUseRule=UseColor_Specified, specifiedColor=<rgba 0 0 0 0>}
      cornerRadii=<0 0 0 0>
     }
     tintColor {colorUseRule=UseColor_Specified, specifiedColor=<rgba 1 1 1 1>}
     uVRegion {bIsValid=false, max=<0 0>, min=<0 0>}
    }
   }
   slot {class=CanvasPanelSlot, properties {layoutData {alignment=<0 0>, anchors {maximum=<0 0>, minimum=<0 0>}, offsets=<ltrb 0 0 3840 1080>}}}
  }
  Panel2Chr1Image {
   widgetClassPath=&Image
   bIsVariable=true
   parent=Panel2
   widget=Panel2Chr1Image
   properties {
    clipping=ClipToBounds
    visibility=HitTestInvisible
    brush {
     bIsDynamicallyLoaded=false
     drawAs=Image
     imageType=FullColor
     mirroring=NoMirror
     resourceName=None
     resourceObject=null
     tiling=NoTile
     imageSize=<368 328>
     margin=<ltrb 0 0 0 0>
     outlineSettings {
      bUseBrushTransparency=false
      roundingType=HalfHeightRadius
      width=0
      color {colorUseRule=UseColor_Specified, specifiedColor=<rgba 0 0 0 0>}
      cornerRadii=<0 0 0 0>
     }
     tintColor {colorUseRule=UseColor_Specified, specifiedColor=<rgba 1 1 1 1>}
     uVRegion {bIsValid=false, max=<0 0>, min=<0 0>}
    }
   }
   slot {
    class=CanvasPanelSlot
    properties {layoutData {alignment=<0 0>, anchors {maximum=<0 0>, minimum=<0 0>}, offsets=<ltrb 256.4548 1071.6748 475 423.3>}}
   }
  }
  Panel2Chr2Image {
   widgetClassPath=&Image
   bIsVariable=true
   parent=Panel2
   widget=Panel2Chr2Image
   properties {
    clipping=ClipToBounds
    visibility=HitTestInvisible
    brush {
     bIsDynamicallyLoaded=false
     drawAs=Image
     imageType=FullColor
     mirroring=NoMirror
     resourceName=None
     resourceObject=null
     tiling=NoTile
     imageSize=<530 630.9524>
     margin=<ltrb 0 0 0 0>
     outlineSettings {
      bUseBrushTransparency=false
      roundingType=HalfHeightRadius
      width=0
      color {colorUseRule=UseColor_Specified, specifiedColor=<rgba 0 0 0 0>}
      cornerRadii=<0 0 0 0>
     }
     tintColor {colorUseRule=UseColor_Specified, specifiedColor=<rgba 1 1 1 1>}
     uVRegion {bIsValid=false, max=<0 0>, min=<0 0>}
    }
   }
   slot {
    class=CanvasPanelSlot
    properties {layoutData {alignment=<0 0>, anchors {maximum=<0 0>, minimum=<0 0>}, offsets=<ltrb 840.4142 976.4716 530 630.9524>}}
   }
  }
  Panel2Chr3Image {
   widgetClassPath=&Image
   bIsVariable=true
   parent=Panel2
   widget=Panel2Chr3Image
   properties {
    clipping=ClipToBounds
    visibility=HitTestInvisible
    brush {
     bIsDynamicallyLoaded=false
     drawAs=Image
     imageType=FullColor
     mirroring=NoMirror
     resourceName=None
     resourceObject=null
     tiling=NoTile
     imageSize=<510 685.4839>
     margin=<ltrb 0 0 0 0>
     outlineSettings {
      bUseBrushTransparency=false
      roundingType=HalfHeightRadius
      width=0
      color {colorUseRule=UseColor_Specified, specifiedColor=<rgba 0 0 0 0>}
      cornerRadii=<0 0 0 0>
     }
     tintColor {colorUseRule=UseColor_Specified, specifiedColor=<rgba 1 1 1 1>}
     uVRegion {bIsValid=false, max=<0 0>, min=<0 0>}
    }
   }
   slot {
    class=CanvasPanelSlot
    properties {layoutData {alignment=<0 0>, anchors {maximum=<0 0>, minimum=<0 0>}, offsets=<ltrb 1403.8569 947.6832 510 685.4839>}}
   }
  }
  Panel2Chr4Image {
   widgetClassPath=&Image
   bIsVariable=true
   parent=Panel2
   widget=Panel2Chr4Image
   properties {
    clipping=ClipToBounds
    visibility=HitTestInvisible
    brush {
     bIsDynamicallyLoaded=false
     drawAs=Image
     imageType=FullColor
     mirroring=NoMirror
     resourceName=None
     resourceObject=null
     tiling=NoTile
     imageSize=<530 885.1>
     margin=<ltrb 0 0 0 0>
     outlineSettings {
      bUseBrushTransparency=false
      roundingType=HalfHeightRadius
      width=0
      color {colorUseRule=UseColor_Specified, specifiedColor=<rgba 0 0 0 0>}
      cornerRadii=<0 0 0 0>
     }
     tintColor {colorUseRule=UseColor_Specified, specifiedColor=<rgba 1 1 1 1>}
     uVRegion {bIsValid=false, max=<0 0>, min=<0 0>}
    }
   }
   slot {
    class=CanvasPanelSlot
    properties {layoutData {alignment=<0 0>, anchors {maximum=<0 0>, minimum=<0 0>}, offsets=<ltrb 1998.1309 698.1177 530 885.1>}}
   }
  }
  Panel2Chr5Image {
   widgetClassPath=&Image
   bIsVariable=true
   parent=Panel2
   widget=Panel2Chr5Image
   properties {
    clipping=ClipToBounds
    visibility=HitTestInvisible
    brush {
     bIsDynamicallyLoaded=false
     drawAs=Image
     imageType=FullColor
     mirroring=NoMirror
     resourceName=None
     resourceObject=null
     tiling=NoTile
     imageSize=<513 529.3723>
     margin=<ltrb 0 0 0 0>
     outlineSettings {
      bUseBrushTransparency=false
      roundingType=HalfHeightRadius
      width=0
      color {colorUseRule=UseColor_Specified, specifiedColor=<rgba 0 0 0 0>}
      cornerRadii=<0 0 0 0>
     }
     tintColor {colorUseRule=UseColor_Specified, specifiedColor=<rgba 1 1 1 1>}
     uVRegion {bIsValid=false, max=<0 0>, min=<0 0>}
    }
   }
   slot {
    class=CanvasPanelSlot
    properties {layoutData {alignment=<0 0>, anchors {maximum=<0 0>, minimum=<0 0>}, offsets=<ltrb 2573.1636 1020.9061 513 529.3723>}}
   }
  }
  Panel2Chr6Image {
   widgetClassPath=&Image
   bIsVariable=true
   parent=Panel2
   widget=Panel2Chr6Image
   properties {
    clipping=ClipToBounds
    visibility=HitTestInvisible
    brush {
     bIsDynamicallyLoaded=false
     drawAs=Image
     imageType=FullColor
     mirroring=NoMirror
     resourceName=None
     resourceObject=null
     tiling=NoTile
     imageSize=<560 800.8>
     margin=<ltrb 0 0 0 0>
     outlineSettings {
      bUseBrushTransparency=false
      roundingType=HalfHeightRadius
      width=0
      color {colorUseRule=UseColor_Specified, specifiedColor=<rgba 0 0 0 0>}
      cornerRadii=<0 0 0 0>
     }
     tintColor {colorUseRule=UseColor_Specified, specifiedColor=<rgba 1 1 1 1>}
     uVRegion {bIsValid=false, max=<0 0>, min=<0 0>}
    }
   }
   slot {
    class=CanvasPanelSlot
    properties {layoutData {alignment=<0 0>, anchors {maximum=<0 0>, minimum=<0 0>}, offsets=<ltrb 3149.4795 693.0887 560 800.8>}}
   }
  }
  Panel3 {
   widgetClassPath=&CanvasPanel
   parent=CanvasPanel_59
   widget=Panel3
   properties {clipping=ClipToBounds, visibility=HitTestInvisible}
   slot {
    class=CanvasPanelSlot
    properties {zOrder=4, layoutData {alignment=<0 0>, anchors {maximum=<0.5 0.5>, minimum=<0.5 0.5>}, offsets=<ltrb -960 -540 1920 1080>}}
   }
  }
  Panel3BG {
   widgetClassPath=&Image
   bIsVariable=true
   parent=Panel3
   widget=Panel3BG
   properties {
    clipping=ClipToBounds
    visibility=HitTestInvisible
    brush {
     bIsDynamicallyLoaded=false
     drawAs=Image
     imageType=FullColor
     mirroring=NoMirror
     resourceName=None
     resourceObject=null
     tiling=NoTile
     imageSize=<1920 1080>
     margin=<ltrb 0 0 0 0>
     outlineSettings {
      bUseBrushTransparency=false
      roundingType=HalfHeightRadius
      width=0
      color {colorUseRule=UseColor_Specified, specifiedColor=<rgba 0 0 0 0>}
      cornerRadii=<0 0 0 0>
     }
     tintColor {colorUseRule=UseColor_Specified, specifiedColor=<rgba 1 1 1 1>}
     uVRegion {bIsValid=false, max=<0 0>, min=<0 0>}
    }
   }
   slot {
    class=CanvasPanelSlot
    properties {zOrder=1, layoutData {alignment=<0 0>, anchors {maximum=<0.5 0.5>, minimum=<0.5 0.5>}, offsets=<ltrb -960 -540 1920 1080>}}
   }
  }
  P3Chr05 {
   widgetClassPath=&Image
   bIsVariable=true
   parent=Panel3
   widget=P3Chr05
   properties {
    clipping=ClipToBounds
    visibility=HitTestInvisible
    brush {
     bIsDynamicallyLoaded=false
     drawAs=Image
     imageType=FullColor
     mirroring=NoMirror
     resourceName=None
     resourceObject=null
     tiling=NoTile
     imageSize=<372 800>
     margin=<ltrb 0 0 0 0>
     outlineSettings {
      bUseBrushTransparency=false
      roundingType=HalfHeightRadius
      width=0
      color {colorUseRule=UseColor_Specified, specifiedColor=<rgba 0 0 0 0>}
      cornerRadii=<0 0 0 0>
     }
     tintColor {colorUseRule=UseColor_Specified, specifiedColor=<rgba 1 1 1 1>}
     uVRegion {bIsValid=false, max=<0 0>, min=<0 0>}
    }
   }
   slot {
    class=CanvasPanelSlot
    properties {zOrder=1, layoutData {alignment=<0 0>, anchors {maximum=<0.5 0.5>, minimum=<0.5 0.5>}, offsets=<ltrb -140.8582 234.953 372 800>}}
   }
  }
  P3Chr06 {
   widgetClassPath=&Image
   bIsVariable=true
   parent=Panel3
   widget=P3Chr06
   properties {
    clipping=ClipToBounds
    visibility=HitTestInvisible
    brush {
     bIsDynamicallyLoaded=false
     drawAs=Image
     imageType=FullColor
     mirroring=NoMirror
     resourceName=None
     resourceObject=null
     tiling=NoTile
     imageSize=<916 800>
     margin=<ltrb 0 0 0 0>
     outlineSettings {
      bUseBrushTransparency=false
      roundingType=HalfHeightRadius
      width=0
      color {colorUseRule=UseColor_Specified, specifiedColor=<rgba 0 0 0 0>}
      cornerRadii=<0 0 0 0>
     }
     tintColor {colorUseRule=UseColor_Specified, specifiedColor=<rgba 1 1 1 1>}
     uVRegion {bIsValid=false, max=<0 0>, min=<0 0>}
    }
   }
   slot {
    class=CanvasPanelSlot
    properties {zOrder=1, layoutData {alignment=<0 0>, anchors {maximum=<0.5 0.5>, minimum=<0.5 0.5>}, offsets=<ltrb -415.154 196.5076 916 800>}}
   }
  }
  P3Chr02 {
   widgetClassPath=&Image
   bIsVariable=true
   parent=Panel3
   widget=P3Chr02
   properties {
    clipping=ClipToBounds
    visibility=HitTestInvisible
    brush {
     bIsDynamicallyLoaded=false
     drawAs=Image
     imageType=FullColor
     mirroring=NoMirror
     resourceName=None
     resourceObject=null
     tiling=NoTile
     imageSize=<600 800>
     margin=<ltrb 0 0 0 0>
     outlineSettings {
      bUseBrushTransparency=false
      roundingType=HalfHeightRadius
      width=0
      color {colorUseRule=UseColor_Specified, specifiedColor=<rgba 0 0 0 0>}
      cornerRadii=<0 0 0 0>
     }
     tintColor {colorUseRule=UseColor_Specified, specifiedColor=<rgba 1 1 1 1>}
     uVRegion {bIsValid=false, max=<0 0>, min=<0 0>}
    }
   }
   slot {
    class=CanvasPanelSlot
    properties {zOrder=1, layoutData {alignment=<0 0>, anchors {maximum=<0.5 0.5>, minimum=<0.5 0.5>}, offsets=<ltrb -262.7303 156.5201 600 800>}}
   }
  }
  P3Chr03 {
   widgetClassPath=&Image
   bIsVariable=true
   parent=Panel3
   widget=P3Chr03
   properties {
    clipping=ClipToBounds
    visibility=HitTestInvisible
    brush {
     bIsDynamicallyLoaded=false
     drawAs=Image
     imageType=FullColor
     mirroring=NoMirror
     resourceName=None
     resourceObject=null
     tiling=NoTile
     imageSize=<360 800>
     margin=<ltrb 0 0 0 0>
     outlineSettings {
      bUseBrushTransparency=false
      roundingType=HalfHeightRadius
      width=0
      color {colorUseRule=UseColor_Specified, specifiedColor=<rgba 0 0 0 0>}
      cornerRadii=<0 0 0 0>
     }
     tintColor {colorUseRule=UseColor_Specified, specifiedColor=<rgba 1 1 1 1>}
     uVRegion {bIsValid=false, max=<0 0>, min=<0 0>}
    }
   }
   slot {
    class=CanvasPanelSlot
    properties {zOrder=1, layoutData {alignment=<0 0>, anchors {maximum=<0.5 0.5>, minimum=<0.5 0.5>}, offsets=<ltrb -117.7005 116.8471 360 800>}}
   }
  }
  P3Chr04 {
   widgetClassPath=&Image
   bIsVariable=true
   parent=Panel3
   widget=P3Chr04
   properties {
    clipping=ClipToBounds
    visibility=HitTestInvisible
    brush {
     bIsDynamicallyLoaded=false
     drawAs=Image
     imageType=FullColor
     mirroring=NoMirror
     resourceName=None
     resourceObject=null
     tiling=NoTile
     imageSize=<296 800>
     margin=<ltrb 0 0 0 0>
     outlineSettings {
      bUseBrushTransparency=false
      roundingType=HalfHeightRadius
      width=0
      color {colorUseRule=UseColor_Specified, specifiedColor=<rgba 0 0 0 0>}
      cornerRadii=<0 0 0 0>
     }
     tintColor {colorUseRule=UseColor_Specified, specifiedColor=<rgba 1 1 1 1>}
     uVRegion {bIsValid=false, max=<0 0>, min=<0 0>}
    }
   }
   slot {
    class=CanvasPanelSlot
    properties {zOrder=1, layoutData {alignment=<0 0>, anchors {maximum=<0.5 0.5>, minimum=<0.5 0.5>}, offsets=<ltrb -82.7498 192.5873 296 800>}}
   }
  }
  P3Chr01 {
   widgetClassPath=&Image
   bIsVariable=true
   parent=Panel3
   widget=P3Chr01
   properties {
    clipping=ClipToBounds
    visibility=HitTestInvisible
    brush {
     bIsDynamicallyLoaded=false
     drawAs=Image
     imageType=FullColor
     mirroring=NoMirror
     resourceName=None
     resourceObject=null
     tiling=NoTile
     imageSize=<400 800>
     margin=<ltrb 0 0 0 0>
     outlineSettings {
      bUseBrushTransparency=false
      roundingType=HalfHeightRadius
      width=0
      color {colorUseRule=UseColor_Specified, specifiedColor=<rgba 0 0 0 0>}
      cornerRadii=<0 0 0 0>
     }
     tintColor {colorUseRule=UseColor_Specified, specifiedColor=<rgba 1 1 1 1>}
     uVRegion {bIsValid=false, max=<0 0>, min=<0 0>}
    }
   }
   slot {
    class=CanvasPanelSlot
    properties {zOrder=1, layoutData {alignment=<0 0>, anchors {maximum=<0.5 0.5>, minimum=<0.5 0.5>}, offsets=<ltrb -175.4666 78.2701 400 800>}}
   }
  }
 ]
}
